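; BP_Force_PostFX is a short-lived post-process effect actor for the Lightning item FX, producing a timed blur pulse tied to its location.
; After a delay, it creates a dynamic material, passes in its world position, and applies it as a post-process blendable.
; A timeline curve scaled by PostFX_Intensity drives BlurControl, then the actor destroys itself.
; The construction script sets a fixed box extent unless Player Pawn is valid and locally controlled.
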(event EventBeginPlay
  (Utilities|FlowControl|Delay (Variables|Default|GetDelayTime))
  (bind _returnvalue (Rendering|Material|CreateDynamicMaterialInstance (Variables|Default|GetM) "None" "None" :!class /Script/Engine.KismetMaterialLibrary))
  (Rendering|Material|SetVectorParameterValue _returnvalue "WorldPosition" (Math|Color|MakeColor (.x (Transformation|GetActorLocation)) (.y (Transformation|GetActorLocation)) (.z (Transformation|GetActorLocation))) :!class /Script/Engine.MaterialInstanceDynamic)
  (Rendering|Material|SetScalarParameterValue _returnvalue "BlurControl" 0.0 :!class /Script/Engine.MaterialInstanceDynamic)
  (Variables|/Script/Engine.PostProcessComponent|SetSettings (Utilities|Struct|MakePostProcessSettings (Utilities|Struct|MakeWeightedBlendables (Utilities|Array|MakeArray (Utilities|Struct|MakeWeightedBlendable 1.0 _returnvalue)))) (Variables|Default|GetPostProcess))
  (Variables|Default|SetDM _returnvalue)
  (CallFunction|StartTimeline))

(event Custom|StartTimeline
  (bind (_direction _postweightcurve) (Timeline|Timeline PlayFromStart)
    (:Update
      (Rendering|Material|SetScalarParameterValue (Variables|Default|GetDM) "BlurControl" (* _postweightcurve (Variables|Default|GetPostFX_Intensity)) :!class /Script/Engine.MaterialInstanceDynamic))
    (:Finished
      (Actor|DestroyActor))))

(fn ConstructionScript
  (if (not (and (Utilities|IsValid (Variables|Default|GetPlayer%20Pawn) :!class /Script/Engine.KismetSystemLibrary) (Pawn|IsLocallyControlled (Variables|Default|GetPlayer%20Pawn))))
    (Components|Box|SetBoxExtent (Variables|Default|GetBox) "2000.000000,2000.000000,2000.000000" true)))
