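; Child post-process effect Blueprint for the lightning item FX that drives a timed blur and then removes itself.
; StartTimeline plays a timeline whose weight curve, scaled by PostFX_Intensity, sets the BlurControl parameter on the dynamic material DM.
; The actor destroys itself when the timeline finishes.
; Construction defers to the parent, while BeginPlay, overlap and Tick hold no active logic.
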(event EventBeginPlay
  (|CallParentFunction) ; This node is disabled and will not be called.\nDrag off pins to build functionality.
)

(event Collision|EventActorBeginOverlap
  (params (OtherActor :type Actor%20Object%20Reference)))

(event EventTick
  (params (DeltaSeconds :type Float%20%28single-precision%29)))

(event EventStartTimeline
  (bind (_direction _postweightcurve) (Timeline|Timeline_0 PlayFromStart)
    (:Update
      (Rendering|Material|SetScalarParameterValue (Variables|Default|GetDM) "BlurControl" (* _postweightcurve (Variables|Default|GetPostFX_Intensity)) :!class /Script/Engine.MaterialInstanceDynamic))
    (:Finished
      (Actor|DestroyActor))))

(fn ConstructionScript
  (|CallParentFunction))
