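; BP_Throw_PostFX is a short-lived post-process actor for the lightning item's throw effect, visible only to the local player.
; After waiting two ticks on BeginPlay, it checks that its Player Pawn is valid and locally controlled, otherwise collapsing the box and destroying itself.
; For the local pawn, it enlarges the box, waits DelayTime, then runs a timeline curve into the post-process blend weight and destroys itself when finished.
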(event Custom|StartTimeline
  (bind (_direction _postweightcurve) (Timeline|Timeline PlayFromStart)
    (:Update
      (Variables|/Script/Engine.PostProcessComponent|SetBlendWeight _postweightcurve (Variables|Default|GetPostProcess)))
    (:Finished
      (Actor|DestroyActor))))

(event EventBeginPlay
  (Utilities|FlowControl|DelayUntilNextTick)
  (Utilities|FlowControl|DelayUntilNextTick)
  (if (and (Utilities|IsValid (Variables|Default|GetPlayer%20Pawn) :!class /Script/Engine.KismetSystemLibrary) (Pawn|IsLocallyControlled (Variables|Default|GetPlayer%20Pawn)))
    (Components|Box|SetBoxExtent (Variables|Default|GetBox) "500.000000,500.000000,500.000000" true)
    (Utilities|FlowControl|Delay (Variables|Default|GetDelayTime))
    (CallFunction|StartTimeline)
    (else
      (Components|Box|SetBoxExtent (Variables|Default|GetBox) "0, 0, 0" true)
      (Actor|DestroyActor))))

(fn ConstructionScript)
