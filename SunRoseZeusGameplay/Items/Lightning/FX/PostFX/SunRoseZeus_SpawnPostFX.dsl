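; Anim notify for the SunRoseZeus Lightning item that spawns a post FX actor on the Fortnite player pawn playing the animation.
; On notify, it casts the mesh owner to FortPlayerPawnAthena and spawns its Class variable at the mesh transform, owned by that pawn.
; It snaps the actor to the mesh, then casts it to BP_Throw_PostFX to set its Player Pawn.
; The attach result is the notify's return value.
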(fn Received_Notify
  (params (MeshComp :type Skeletal%20Mesh%20Component%20Object%20Reference) (Animation :type Anim%20Sequence%20Base%20Object%20Reference) (EventReference :type Anim%20Notify%20Event%20Reference%20Structure%20%28by%20ref%29))
  (bind _asfort_player_pawn_athena (Utilities|Casting|CastToFortPlayerPawnAthena (Components|GetOwner MeshComp)))
  (bind _returnvalue (Game|SpawnActorfromClass (Variables|Default|GetClass) (Transformation|GetWorldTransform MeshComp) "Undefined" "MultiplyWithRoot" 0 _asfort_player_pawn_athena))
  (bind _returnvalue_2 (Transformation|AttachActorToComponent _returnvalue MeshComp "None" "SnapToTarget" "SnapToTarget" "KeepWorld" true))
  (bind _asbp_throw_post_fx (Utilities|Casting|CastToBP_Throw_PostFX _returnvalue))
  (Variables|/SunRoseZeusGameplay/Items/Lightning/FX/PostFX/BP_Throw_PostFX.BP_Throw_PostFX_C|SetPlayer%20Pawn _asfort_player_pawn_athena _asbp_throw_post_fx)
  (return _returnvalue_2))
